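# Cloth-shaded anime material for the Magenta skin of the IndieBucket pickaxe, built on shared anime master functions.
# Character PreFX feeds anime shading, then a chain of visor, WPO, pickup-shrink, mouth flipbook and skin FX functions before PostFX.
# Before PostFX, a custom emissive edit blends in a stepped shadow color, driven by ShadowHue, CurveColorInfluence and hue shifting, masked by a skin-specific ShadowFx texture.
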Material /BRCosmetics/Weapons/FORT_Melee/Pickaxe_IndieBucket/Skins/Magenta/Materials/M_Indiebucket_Pickaxe_AnimeShading
graph {
 properties {
  bAutomaticallySetUsageInEditor=false
  bUseMaterialAttributes=true
  bUsedWithMeshDeformer=true
  bUsedWithSkeletalMesh=true
  pixelDepthOffsetMode=PDOM_Legacy
  shadingModel=MSM_Cloth
  usedShadingModels=Cloth
 }
 nodes [
  n1 = MaterialFunctionCall() {materialFunction=&BaseFN_CharacterFunction_PreFX}
  n2 = MaterialFunctionCall!(
   Material=@n1.Result
   LitDiffuseOverride
   SSC_AlphaOverride
   Tint Mask
   DF InkLine Visibility Mask
   DF Inkline Injection
   DF Inkline Color Override
   DF Inkline Color Override Mask
   CustomUVs
  ) {materialFunction=/Game/Packages/Fortress_SharedMaterials/Anime_Master_Material/MF_AnimeShading}
  n3 = MaterialFunctionCall!(In=@n2.Result) {materialFunction=/Game/Packages/Fortress_SharedMaterials/Anime_Master_Material/MF_AnimatedVisor}
  n4 = MaterialFunctionCall!(In=MaterialFunctionCall!(In=MaterialFunctionCall!(In=@n3.Result) {materialFunction=/Game/Characters/Player/Female/Medium/Bodies/F_MED_Peppermint/Materials/MF_HairPuffWPO}.Result) {materialFunction=/Game/Packages/Fortress_SharedMaterials/Anime_Master_Material/MF_BobbingWPO}.Result) {materialFunction=/Game/Packages/Fortress_SharedMaterials/Anime_Master_Material/MF_PickupShrinking}
  n5 = MaterialFunctionCall!(
   MA=MaterialFunctionCall!(
    Material=MaterialFunctionCall!(In=@n4.Result) {materialFunction=/Game/Packages/Fortress_SharedMaterials/Anime_Master_Material/MF_ToonFlipbookMouth}.Result
    SkinFX_Mask=@n2.FX
   ) {materialFunction=/Game/Global/MaterialFunctions/MF_ClothSimWPO_v2}.Result
   Shading=@n2.Shading
   InkLine Mask=@n2.InkLine Mask
   SSC=@n2.SSC
   SkinFX=@n2.FX
  ) {materialFunction=/Game/Packages/Fortress_SharedMaterials/Anime_Master_Material/MF_AnimeSkinFX}
  Composite!(None=@n5.MA)
  MaterialFunctionCall!(Material) {materialFunction=/Game/Packages/Fortress_SharedMaterials/Anime_Master_Material/MF_ToonOutline_Mobile}
  n6 = ScalarParameter() {group=IndieBucket, parameterName=CurveColorInfluence}
  n7 = NamedRerouteUsage()
  MaterialFunctionCall(Hue Shift Percentage=@n6, Texture=@n7) {materialFunction=&HueShift}
  MaterialExpressionReroute_6 = Reroute(None=@n5.MA)
  n8 = GetMaterialAttributes(None=@MaterialExpressionReroute_6.Input) {attributeGetTypes [#1]}
  n9 = LinearInterpolate(
   A=NamedRerouteUsage()
   B=VectorParameter() {group=IndieBucket, parameterName=ShadowHue, defaultValue=<rgba 0.3185 0.0185 0.2232 1>}.RGB
   Alpha=@n6
  )
  n10 = Step(X=NamedRerouteUsage()) {constY=0.1}
  n11 = SetMaterialAttributes(
   MaterialAttributes=@n8.MaterialAttributes
   Emissive Color=LinearInterpolate(
    A=NamedRerouteUsage()
    B=LinearInterpolate(A=@n9, B=@n7, Alpha=@n10)
    Alpha=TextureSample() {texture=/BRCosmetics/Weapons/FORT_Melee/Pickaxe_IndieBucket/Skins/Magenta/Textures/T_Pickaxe_IndieBucket_Magenta_ShadowFx}.R
   )
  ) {attributeSetTypes [#1]}
  MaterialExpressionReroute_5 = Reroute(None=@n11)
  MaterialFunctionCall(
   AO=@n1.AO
   Distance Blend=@n1.AO
   Fresnel=@n1.AO
   Gmap=@n1.AO
   Pre Skinned Local Position=@n1.AO
   Material=@MaterialExpressionReroute_5.Output
  ) {materialFunction=&BaseFN_CharacterFunction_PostFX}
  NamedRerouteDeclaration(None=ComponentMask(None=@n2.Shading) {r=true}) {name=ShadingCurve, nodeColor=<rgba 0.875 1 0 1>}
  NamedRerouteDeclaration(None=@n8.Emissive Color) {name=Emissive, nodeColor=<rgba 0 1 0.0391 1>}
  PinBase() {reroutePins [Input {expression=&self:MaterialExpressionReroute_6}]}
  PinBase!(Output=@n11) {reroutePins [Output {expression=&self:MaterialExpressionReroute_5}]}
 ]
}
